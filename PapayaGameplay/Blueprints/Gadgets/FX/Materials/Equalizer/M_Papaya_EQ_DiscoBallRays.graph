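# Translucent unlit Niagara sprite material for the disco ball rays of the Papaya equalizer gadget effects.
# A radial ramp and slowly scrolling time, offset per particle, sample noise that is masked by a spikeball flare and depth-faded.
# Equalizer audio amplitude thresholds a wing ramp, blending pale blue-grey into the collection-driven gadget color.
# The same noise slightly hue-shifts that color before HDR gain and emissive adjustment.
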Material /PapayaGameplay/Blueprints/Gadgets/FX/Materials/Equalizer/M_Papaya_EQ_DiscoBallRays
graph {
 properties {bUsedWithNiagaraSprites=true, blendMode=BLEND_Translucent, pixelDepthOffsetMode=PDOM_Legacy, shadingModel=MSM_Unlit, usedShadingModels=Unlit}
 nodes [
  n1 = TextureSample() {samplerType=SAMPLERTYPE_LinearGrayscale, texture=/PapayaGameplay/Blueprints/Gadgets/FX/Textures/EqualizerComponents_v2_RadialRamp}
  n2 = Time()
  n3 = Constant() {r=0.025}
  n4 = Add(A=Frac(None=Multiply(A=@n2, B=@n3)), B=DynamicParameter() {paramNames [MaterialRandom, Param2, Param3, Param4]}.MaterialRandom)
  n5 = Desaturation(
   None=TextureSample(UVs=MaterialFunctionCall(X=@n1.R, Y=@n4) {materialFunction=&MakeFloat2}.Result) {samplerType=SAMPLERTYPE_LinearColor, texture=&HighContrastNoise}.RGB
   Fraction=Constant() {r=0.8}
  )
  n6 = TextureSample() {texture=/Game/Effects/Fort_Effects/Textures/Flares/T_Flare_Spikeball}
  LinearInterpolate(B=Multiply(A=@n5, B=Power(Base=@n6.R) {constExponent=0.7}), Alpha=DepthFade(FadeDistance=Constant() {r=50}))
  n7 = Multiply(A=@n5, B=Constant() {r=0.1})
  n8 = CollectionParameter() {
   collection=&FortniteMaterialParameters
   parameterName=PapayaBaseColorParameter
  }
  n9 = CollectionParameter() {
   collection=&FortniteMaterialParameters
   parameterName=PapayaBaseColorParameterOffset
  }
  n10 = TextureSample() {samplerType=SAMPLERTYPE_LinearGrayscale, texture=/PapayaGameplay/Blueprints/Gadgets/FX/Textures/EqualizerComponents_v5_RadialWingRamp}
  n11 = OneMinus(None=MaterialFunctionCall!(
    Audio Lookup Param=Constant() {r=0.5}
    Average Amplitude Mix=Constant()
    Ramp Texture
    Ramp Exponent=Constant() {r=1}
    Noise Texture
    Noise Scale
    Noise Gain
    Strip Noise Gain
    Grid Tiles X/Y=Constant2Vector() {g=10, r=1}
    Grid Speed
    Output Solid Bars
    TexCoord Override
   ) {materialFunction=/Game/Environments/Papaya/Materials/MaterialFunctions/MF_Papaya_Equalizer_Data}.EQ Raw Amplitude)
  n12 = Add(A=@n11) {constB=0.05}
  n13 = Constant()
  n14 = Constant() {r=1}
  n15 = LinearInterpolate(
   A=Constant3Vector() {constant=<rgba 0.1385 0.1669 0.2 0>}
   B=MaterialFunctionCall!(
    Media Player TexCoord
    Media Color Code Coordinate
    Media Texture Resolution
    Base Color Parameter=Add(A=@n8, B=@n9)
    Base Color Mix
    Media Color Code Mix
    Intermission Color Parameter
    Intermission Color Mix
    Desaturation
   ) {materialFunction=/Game/Environments/Papaya/Materials/MaterialFunctions/MF_Papaya_GadgetColor}.Result
   Alpha=Saturate(None=MaterialFunctionCall(Input=@n10.R, Input Low=@n11, Input High=@n12, Target Low=@n13, Target High=@n14) {materialFunction=&RemapValueRange}.Result)
  )
  MaterialFunctionCall(Emissive=MaterialFunctionCall(
    LDR Input=MaterialFunctionCall(Hue Shift Percentage=@n7, Texture=@n15) {materialFunction=&HueShift}.Result
    HDR Intensity Multiplier=ScalarParameter() {defaultValue=2.5, parameterName=HDR Gain}
   ) {materialFunction=&DeriveHDRfromLDR}.RemappedRange) {materialFunction=&MF_FX_EmissiveAdjustments}
 ]
}
