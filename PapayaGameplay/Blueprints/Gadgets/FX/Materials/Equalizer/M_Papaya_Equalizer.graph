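# Translucent unlit, two-sided material for Papaya gadget FX that draws an equalizer, usable on Niagara ribbons.
# Mirrored UVs feed the shared equalizer data function, which shapes bars from ramp and noise textures, plus strips and top edges.
# The bar value, faded at U edges and by depth, masks a tiled box texture.
# Collection-driven gadget color is hue-shifted by that value, with brightened strips and edges boosting the HDR glow.
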Material /PapayaGameplay/Blueprints/Gadgets/FX/Materials/Equalizer/M_Papaya_Equalizer
graph {
 properties {
  bUsedWithNiagaraRibbons=true
  blendMode=BLEND_Translucent
  pixelDepthOffsetMode=PDOM_Legacy
  shadingModel=MSM_Unlit
  twoSided=true
  usedShadingModels=Unlit
 }
 nodes [
  n1 = TextureCoordinate()
  n2 = MaterialFunctionCall!(
   Repeat Count=MaterialFunctionCall(X=ScalarParameter() {defaultValue=1, parameterName=Mirror Count}, Y=Constant() {r=1}) {materialFunction=&MakeFloat2}.Result
   texcoord=@n1
  ) {materialFunction=/Game/Environments/Papaya/Materials/MaterialFunctions/MF_UVMirror}
  n3 = MaterialFunctionCall(Float2=@n2.Result) {materialFunction=&BreakOutFloat2Components}
  n4 = TextureObjectParameter() {
   parameterName=Ramp Texture
   samplerType=SAMPLERTYPE_LinearGrayscale
   texture=/Game/Global/Textures/Default/Masks_and_Gradients/EqualizerComponents_v2_Ramp
  }
  n5 = TextureObjectParameter() {
   parameterName=Noise Texture
   samplerType=SAMPLERTYPE_LinearGrayscale
   texture=/Game/Global/Textures/Default/Noise/EqualizerComponents_v2_Noise
  }
  n6 = MaterialFunctionCall(X=ScalarParameter() {defaultValue=10, parameterName=Tiles X}, Y=ScalarParameter() {defaultValue=10, parameterName=Tiles Y}) {materialFunction=&MakeFloat2}
  n7 = MaterialFunctionCall!(
   Audio Lookup Param=@n3.R
   Average Amplitude Mix=ScalarParameter() {parameterName=Average Amplitude Mix}
   Ramp Texture=@n4
   Ramp Exponent=ScalarParameter() {defaultValue=1, parameterName=Ramp Exponent}
   Noise Texture=@n5
   Noise Scale=MaterialFunctionCall(
    X=ScalarParameter() {defaultValue=2, group=Color, parameterName=Noise Scale X, sortPriority=-60}
    Y=ScalarParameter() {defaultValue=2, group=Color, parameterName=Noise Scale Y, sortPriority=-60}
   ) {materialFunction=&MakeFloat2}.Result
   Noise Gain=ScalarParameter() {defaultValue=0.2, group=Color, parameterName=Noise Gain, sortPriority=-60}
   Strip Noise Gain=ScalarParameter() {defaultValue=0.5, group=Color, parameterName=Strip Gain, sortPriority=-60}
   Grid Tiles X/Y=@n6.Result
   Grid Speed=ScalarParameter() {defaultValue=1, parameterName=Grid Speed}
   Output Solid Bars=StaticBoolParameter() {parameterName=EQ Solid Bars}
   TexCoord Override
  ) {materialFunction=/Game/Environments/Papaya/Materials/MaterialFunctions/MF_Papaya_Equalizer_Data}
  n8 = MaterialFunctionCall(Float2=@n1) {materialFunction=&BreakOutFloat2Components}
  n9 = ScalarParameter() {parameterName=EdgeFadeMin}
  n10 = ScalarParameter() {defaultValue=0.0001, parameterName=EdgeFadeMax}
  n11 = MaterialFunctionCall(Float2=@n1) {materialFunction=&BreakOutFloat2Components}
  n12 = OneMinus(None=@n10)
  n13 = OneMinus(None=@n9)
  n14 = Multiply(
   A=Subtract(
    A=MaterialFunctionCall(Alpha=@n8.R, Min=@n9, Max=@n10) {materialFunction=&SmoothStep}.Result
    B=MaterialFunctionCall(Alpha=@n11.R, Min=@n12, Max=@n13) {materialFunction=&SmoothStep}.Result
   )
   B=@n7.EQ Value
  )
  LinearInterpolate(B=@n14, Alpha=DepthFade(FadeDistance=ScalarParameter() {defaultValue=500, parameterName=Depth Fade Distance})) {constB=0}
  n15 = TextureSampleParameter2D(UVs=Multiply(A=@n2.Result, B=@n6.Result)) {
   group=Color
   parameterName=Box Texture
   sortPriority=-130
   texture=/PapayaGameplay/Blueprints/Gadgets/FX/Textures/EqualizerComponents_v2_Button
  }
  n16 = CollectionParameter() {
   collection=&FortniteMaterialParameters
   parameterName=PapayaBaseColorParameter
  }
  n17 = CollectionParameter() {
   collection=&FortniteMaterialParameters
   parameterName=PapayaBaseColorParameterOffset
  }
  n18 = Multiply(A=@n3.R, B=@n17) {constB=0.25}
  n19 = MaterialFunctionCall(
   Hue Shift Percentage=Multiply(A=@n14, B=ScalarParameter() {parameterName=Hue Shift Gain})
   Texture=MaterialFunctionCall!(
    Media Player TexCoord
    Media Color Code Coordinate
    Media Texture Resolution
    Base Color Parameter=Add(A=@n16, B=@n18)
    Base Color Mix
    Media Color Code Mix
    Intermission Color Parameter
    Intermission Color Mix
    Desaturation
   ) {materialFunction=/Game/Environments/Papaya/Materials/MaterialFunctions/MF_Papaya_GadgetColor}.Result
  ) {materialFunction=&HueShift}
  n20 = Max(
   A=Multiply(A=Multiply(A=@n15.R, B=@n14), B=@n19.Result)
   B=Multiply(A=@n19.Result, B=ScalarParameter() {group=Color, parameterName=Background Color Gain, sortPriority=-80})
  )
  n21 = Add(
   A=Multiply(A=Multiply(A=@n14, B=@n7.EQ Noise Strips), B=ScalarParameter() {defaultValue=50, parameterName=Strip Brightness})
   B=Multiply(A=Multiply(A=@n14, B=@n7.EQ Top Edge), B=ScalarParameter() {defaultValue=10, parameterName=Top Brightness})
  )
  MaterialFunctionCall(
   LDR Input=@n20
   HDR Tint=Multiply(A=Add(A=@n20, B=@n21), B=ScalarParameter() {defaultValue=5, group=Color, parameterName=HDRMult, sortPriority=-80})
  ) {materialFunction=&DeriveHDRfromLDR}
  TextureCoordinate()
 ]
}
