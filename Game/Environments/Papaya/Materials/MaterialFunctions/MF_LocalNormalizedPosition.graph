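# Library material function that outputs a pixel's local-space position normalized to the object's bounds, centred on the object position.
# World position and object position are transformed to local space, subtracted, then divided by object bounds corrected for object scale, giving roughly -1 to 1.
# Static bool inputs optionally remap X, Y and Z independently to 0-1 before the components are recombined into one float3.
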MaterialFunction /Game/Environments/Papaya/Materials/MaterialFunctions/MF_LocalNormalizedPosition
graph {
 properties {bExposeToLibrary=true}
 nodes [
  n1 = TransformPosition(None=WorldPosition().XYZ) {transformSourceType=TRANSFORMPOSSOURCE_World}
  n2 = TransformPosition(None=ObjectPositionWS()) {transformSourceType=TRANSFORMPOSSOURCE_World}
  n3 = ObjectBounds()
  n4 = MaterialFunctionCall() {materialFunction=/Engine/Functions/Engine_MaterialFunctions02/WorldPositionOffset/ObjectScale}
  n5 = MaterialFunctionCall(Float3=Divide(A=Subtract(A=@n1, B=@n2), B=Divide(A=@n3, B=@n4.Scale XYZ))) {materialFunction=&BreakOutFloat3Components}
  n6 = Add(A=Multiply(A=@n5.R) {constB=0.5}) {constB=0.5}
  n7 = FunctionInput(Preview=StaticBool()) {bUsePreviewValueAsDefault=true, inputName=Remap X 0-1, inputType=FunctionInput_StaticBool}
  n8 = Add(A=Multiply(A=@n5.G) {constB=0.5}) {constB=0.5}
  n9 = FunctionInput(Preview=StaticBool()) {bUsePreviewValueAsDefault=true, inputName=Remap Y 0-1, inputType=FunctionInput_StaticBool, sortPriority=10}
  n10 = Add(A=Multiply(A=@n5.B) {constB=0.5}) {constB=0.5}
  n11 = FunctionInput(Preview=StaticBool()) {bUsePreviewValueAsDefault=true, inputName=Remap Z 0-1, inputType=FunctionInput_StaticBool, sortPriority=20}
  FunctionOutput(None=MaterialFunctionCall(
    X=StaticSwitch(True=@n6, False=@n5.R, Value=@n7)
    Y=StaticSwitch(True=@n8, False=@n5.G, Value=@n9)
    Z=StaticSwitch(True=@n10, False=@n5.B, Value=@n11)
   ) {materialFunction=&MakeFloat3}.Result)
 ]
}
