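# Masked, two-sided material for Niagara mesh particles and ribbons that draws an audio-reactive equalizer tinted by a media-player video texture, with a blend to an intermission color.
# Jukebox collection Percussion bands, chosen by local position and mixed toward AverageAmplitude, set the threshold on a ramp texture.
# That level also steps a tiled noise grid, which hue-shifts the video color.
# Fill, strip-noise and edge terms feed HDR emissive adjustment functions.
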Material /PapayaGameplay/Blueprints/Gadgets/FX/Materials/Equalizer/M_ShardScreenStreams
graph {
 properties {
  bUsedWithNiagaraMeshParticles=true
  bUsedWithNiagaraRibbons=true
  blendMode=BLEND_Masked
  pixelDepthOffsetMode=PDOM_Legacy
  twoSided=true
  usedShadingModels=Default Lit
 }
 nodes [
  Constant() {r=0.5}
  Constant() {r=0.15}
  n1 = ScalarParameter() {defaultValue=100, group=Grid, parameterName=Tiles Y, sortPriority=-90}
  n2 = MaterialFunctionCall(X=ScalarParameter() {defaultValue=100, group=Grid, parameterName=Tiles X, sortPriority=-100}, Y=@n1) {materialFunction=&MakeFloat2}
  n3 = StaticBool() {value=true}
  n4 = MaterialFunctionCall(Float3=MaterialFunctionCall!(Remap X 0-1=@n3, Remap Y 0-1, Remap Z 0-1=@n3) {materialFunction=/Game/Environments/Papaya/Materials/MaterialFunctions/MF_LocalNormalizedPosition}.Result) {materialFunction=&BreakOutFloat3Components}
  n5 = MaterialFunctionCall(X=@n4.R, Y=OneMinus(None=@n4.B)) {materialFunction=&MakeFloat2}
  MaterialFunctionCall!(TexCoord=@n5.Result, Tiles XY=@n2.Result) {materialFunction=/Game/Environments/Papaya/Materials/MF_GridTexCoord}
  n6 = ScalarParameter() {defaultValue=0.01, group=Color, parameterName=Background Color Gain, sortPriority=-80}
  n7 = MaterialFunctionCall(Float4=CollectionParameter() {collection=/Game/Creative/Devices/Jukebox/Creative_Jukebox_MPC, parameterName=Percussion}) {materialFunction=&BreakOutFloat4Components}
  n8 = If(A=@n4.B, "A > B"=@n7.G, "A < B"=If(A=@n4.B, "A > B"=@n7.B, "A < B"=@n7.B) {constB=0.25}) {constB=0.5}
  n9 = OneMinus(None=LinearInterpolate(
    A=If(A=@n4.B, "A > B"=@n7.R, "A < B"=@n8) {constB=0.75}
    B=CollectionParameter() {collection=&Creative_Jukebox_MPC, parameterName=AverageAmplitude}
    Alpha=ScalarParameter() {defaultValue=1, group=Audio Samples, parameterName=AverageAmplitudeMix, sliderMax=1, sortPriority=-60}
   ))
  n10 = Multiply(A=@n9, B=ScalarParameter() {defaultValue=2, group=Grid, parameterName=Grid Speed, sortPriority=-30})
  n11 = Divide(A=Floor(None=Multiply(A=@n10, B=@n1)), B=@n1)
  n12 = MaterialFunctionCall(X=Constant(), Y=Constant() {r=1}) {materialFunction=&MakeFloat2}
  n13 = TextureSampleParameter2D(UVs=MaterialFunctionCall(TexCoord=Panner(Coordinate=@n5.Result, Time=@n11, Speed=@n12.Result) {bFractionalPart=true}, Tiles XY=@n2.Result) {materialFunction=&MF_GridTexCoord}.TexCoord) {
   group=Grid
   parameterName=Noise Texture
   samplerType=SAMPLERTYPE_LinearGrayscale
   sortPriority=-70
   texture=/Game/Global/Textures/Default/Noise/EqualizerComponents_v2_Noise
  }
  n14 = LinearInterpolate(
   A=TextureSample(UVs=@n5.Result) {texture=/Game/Movies/Season5/Papaya_MediaPlayer_Video1}.RGB
   B=VectorParameter() {parameterName=Intermission Color, defaultValue=<rgba 0.2372 0.1055 0.2396 1>}.RGB
   Alpha=ScalarParameter() {parameterName=Intermission}
  )
  n15 = MaterialFunctionCall(
   Hue Shift Percentage=Multiply(A=Subtract(A=@n13.R) {constB=0.5}, B=ScalarParameter() {defaultValue=1, parameterName=Hue Shift Gain})
   Texture=@n14
  ) {materialFunction=&HueShift}
  n16 = TextureSampleParameter2D(UVs=@n5.Result) {
   group=Grid
   parameterName=Ramp Texture
   samplerType=SAMPLERTYPE_LinearGrayscale
   sortPriority=-80
   texture=/Game/Global/Textures/Default/Masks_and_Gradients/EqualizerComponents_v2_Ramp
  }
  n17 = ScalarParameter() {defaultValue=0.5, group=Grid, parameterName=Ramp Exponent, sortPriority=-60}
  n18 = Saturate(None=MaterialFunctionCall(
    Input=Power(Base=@n16.R, Exp=@n17) {constExponent=0.5}
    Input Low=Add(A=@n9) {constB=-0.1}
    Input High=Add(A=@n9) {constB=0.1}
    Target Low=Constant()
    Target High=Constant() {r=1}
   ) {materialFunction=&RemapValueRange}.Result)
  n19 = Constant3Vector() {constant=<rgba 1 1 1 0>}
  n20 = Max(A=Multiply(A=Multiply(A=@n15.Result, B=@n6), B=@n18), B=Multiply(A=@n15.Result, B=Multiply(A=@n19, B=@n18)))
  n21 = Subtract(A=@n18) {constB=0.5}
  n22 = OneMinus(None=Multiply(A=Abs(None=@n21)) {constB=2})
  n23 = ScalarParameter() {defaultValue=5, group=Color, parameterName=Top Edge Gain, sortPriority=-70}
  n24 = MaterialFunctionCall(Float2=@n2.Result) {materialFunction=&BreakOutFloat2Components}
  n25 = Constant() {r=1}
  n26 = TextureSampleParameter2D(UVs=MaterialFunctionCall(
    TexCoord=@n5.Result
    Tiles XY=MaterialFunctionCall(X=@n24.R, Y=@n25) {materialFunction=&MakeFloat2}.Result
   ) {materialFunction=&MF_GridTexCoord}.TexCoord) {
   group=Grid
   parameterName=Strip Noise Texture
   samplerType=SAMPLERTYPE_LinearGrayscale
   sortPriority=-65
   texture=&EqualizerComponents_v2_Noise
  }
  n27 = OneMinus(None=@n26.R)
  n28 = ScalarParameter() {defaultValue=1, group=Grid, parameterName=Strip Noise Gain, sortPriority=-40}
  n29 = Multiply(A=OneMinus(None=Multiply(A=@n27, B=@n28) {constB=0.1}), B=@n9)
  n30 = ScalarParameter() {defaultValue=5, group=Color, parameterName=Strip Gain, sortPriority=-60}
  n31 = Add(A=@n20, B=Add(A=Multiply(A=@n29, B=@n30), B=Multiply(A=@n22, B=@n23) {constB=20}))
  n32 = ScalarParameter() {defaultValue=5, group=Color, parameterName=HDRMult, sortPriority=-80}
  MaterialFunctionCall(Emissive=MaterialFunctionCall(LDR Input=@n20, HDR Tint=Multiply(A=@n31, B=@n32)) {materialFunction=&DeriveHDRfromLDR}.RemappedRange) {materialFunction=&MF_FX_EmissiveAdjustments}
  Reroute(None=@n14)
 ]
}
