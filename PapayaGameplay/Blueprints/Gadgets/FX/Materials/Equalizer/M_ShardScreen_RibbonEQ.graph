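# Additive, unlit, two-sided material for an audio-reactive equalizer ribbon in Papaya gadget FX.
# SplineThicken width follows the event audio amplitude and camera distance, while time-driven sine waves sway the ribbon perpendicular to the view via world position offset.
# Panning noise, center and tiled square ramps are tinted by the shared gadget color function using quantized UVs.
# Equalizer data boosts the HDR glow, which depth fade softens at intersections.
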Material /PapayaGameplay/Blueprints/Gadgets/FX/Materials/Equalizer/M_ShardScreen_RibbonEQ
graph {
 properties {blendMode=BLEND_Additive, pixelDepthOffsetMode=PDOM_Legacy, shadingModel=MSM_Unlit, twoSided=true, usedShadingModels=Unlit}
 nodes [
  n1 = CrossProduct(
   A=MaterialFunctionCall() {materialFunction=&CameraDirectionVector}.Result
   B=Constant3Vector() {constant=<rgba 0 0 1 0>}
  )
  n2 = MaterialFunctionCall(Float2=TextureCoordinate()) {materialFunction=&BreakOutFloat2Components}
  n3 = Multiply(A=@n2.G) {constB=5}
  n4 = Time()
  n5 = Multiply(A=@n4) {constB=0.5}
  n6 = Multiply(A=@n2.G) {constB=6}
  n7 = Multiply(A=@n4) {constB=0.3}
  n8 = Add(A=Sine(None=Add(A=@n3, B=@n5)), B=Sine(None=Add(A=@n6, B=@n7)))
  n9 = CollectionParameter() {collection=&FortniteMaterialParameters, parameterName=EventAudio_Amplitude}
  n10 = LinearInterpolate(
   A=ScalarParameter() {defaultValue=0.05, parameterName=SplineThicknessBeatFar}
   B=ScalarParameter() {defaultValue=2.5, parameterName=SplineThicknessBeatClose}
   Alpha=@n9
  )
  n11 = CameraPositionWS()
  n12 = WorldPosition()
  n13 = Divide(
   A=Subtract(A=Distance(A=@n11, B=@n12.XYZ), B=ScalarParameter() {defaultValue=256, parameterName=Distance Add})
   B=ScalarParameter() {defaultValue=1, parameterName=Scale Distance Divisor}
  )
  n14 = LinearInterpolate(
   A=Multiply(A=ScalarParameter() {defaultValue=300, parameterName=SplineWidthLow}, B=@n10)
   B=Multiply(A=ScalarParameter() {defaultValue=350, parameterName=SplineWidthHigh}, B=@n10)
   Alpha=OneMinus(None=Clamp(None=@n13))
  )
  Add(
   A=MaterialFunctionCall(WidthBase=@n14, WidthTip=@n14) {materialFunction=&SplineThicken}.WorldPositionOffset
   B=Multiply(A=Multiply(A=@n1, B=@n8), B=ScalarParameter() {defaultValue=50, parameterName=WPO Gain})
  )
  n15 = TextureCoordinate()
  n16 = MaterialFunctionCall(Float2=@n15) {materialFunction=&BreakOutFloat2Components}
  n17 = Time()
  n18 = MaterialFunctionCall(X=Multiply(A=@n16.R) {constB=0.11}, Y=Add(A=@n16.G, B=Multiply(A=@n17) {constB=0.2})) {materialFunction=&MakeFloat2}
  n19 = TextureSample(UVs=@n18.Result) {samplerType=SAMPLERTYPE_LinearColor, texture=&HighContrastNoise}
  Desaturation(
   None=Add(A=@n19.R, B=TextureSample(UVs=@n18.Result) {texture=/Game/Athena/Prototype/Blueprints/Sneezy/Textures/T_Sneezy_MagneticWaves}.R)
   Fraction=Constant() {r=0.8}
  )
  n20 = Multiply(
   A=@n19.R
   B=TextureSample(UVs=MaterialFunctionCall(X=@n16.G, Y=@n16.G) {materialFunction=&MakeFloat2}.Result) {samplerType=SAMPLERTYPE_LinearGrayscale, texture=/PapayaGameplay/Blueprints/Gadgets/FX/Textures/EqualizerComponents_v2_CenterRamp}.R
  )
  n21 = CollectionParameter() {
   collection=&FortniteMaterialParameters
   parameterName=PapayaBaseColorParameter
  }
  n22 = Constant() {r=150}
  n23 = MaterialFunctionCall!("Param (0-1)"=@n16.G, Tiles=@n22, Apply Offset) {materialFunction=/Game/Environments/Papaya/Materials/MaterialFunctions/MF_Papaya_Quantize_Scalar}
  n24 = MaterialFunctionCall!(
   Media Player TexCoord
   Media Color Code Coordinate
   Media Texture Resolution
   Base Color Parameter=Add(A=@n21, B=Multiply(A=@n23.Result))
   Base Color Mix
   Media Color Code Mix
   Intermission Color Parameter
   Intermission Color Mix
   Desaturation
  ) {materialFunction=/Game/Environments/Papaya/Materials/MaterialFunctions/MF_Papaya_GadgetColor}
  n25 = Multiply(A=@n20, B=@n24.Result)
  n26 = MaterialFunctionCall(X=Constant() {r=5}, Y=@n22) {materialFunction=&MakeFloat2}
  n27 = Multiply(A=@n15, B=@n26.Result)
  n28 = Multiply(
   A=Multiply(
    A=@n20
    B=TextureSample(UVs=@n27) {samplerType=SAMPLERTYPE_LinearColor, texture=/PapayaGameplay/Blueprints/Gadgets/FX/Textures/EqualizerComponents_v2_SquareRamp}.R
   )
   B=LinearInterpolate(Alpha=Power(Base=@n16.G) {constExponent=0.5})
  )
  n29 = Saturate(None=Add(A=@n25, B=Multiply(A=@n24.Result, B=@n28)))
  n30 = Constant()
  n31 = MaterialFunctionCall!(
   Audio Lookup Param=@n16.R
   Average Amplitude Mix=Constant() {r=0.2}
   Ramp Texture
   Ramp Exponent=Constant() {r=1}
   Noise Texture
   Noise Scale
   Noise Gain=@n30
   Strip Noise Gain=@n30
   Grid Tiles X/Y=@n26.Result
   Grid Speed
   Output Solid Bars
   TexCoord Override
  ) {materialFunction=/Game/Environments/Papaya/Materials/MaterialFunctions/MF_Papaya_Equalizer_Data}
  n32 = Add(A=@n29, B=Multiply(A=@n31.EQ Value, B=ScalarParameter() {defaultValue=10, parameterName=EQ Boost}))
  n33 = ScalarParameter() {defaultValue=20, parameterName=HDR Gain}
  LinearInterpolate(
   B=MaterialFunctionCall(LDR Input=@n29, HDR Tint=Multiply(A=@n32, B=@n33)) {materialFunction=&DeriveHDRfromLDR}.RemappedRange
   Alpha=DepthFade(FadeDistance=Constant() {r=500})
  )
  MaterialFunctionCall(X=Constant() {r=0.1}, Y=@n16.G) {materialFunction=&MakeFloat2}
  n34 = MaterialFunctionCall(
   Input=MaterialFunctionCall(Float2=TextureCoordinate()) {materialFunction=&BreakOutFloat2Components}.G
   Input Low=@n31.EQ Raw Amplitude
   Input High=Add(A=@n31.EQ Raw Amplitude) {constB=0.025}
   Target Low=Constant()
   Target High=Constant() {r=1}
  ) {materialFunction=&RemapValueRange}
  Multiply(A=Add(A=Saturate(None=@n34.Result)), B=@n14)
  Multiply(A=@n25, B=@n31.EQ Value)
 ]
}
